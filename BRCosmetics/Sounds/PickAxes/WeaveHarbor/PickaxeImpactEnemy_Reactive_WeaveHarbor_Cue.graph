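# Sound cue for the WeaveHarbor cosmetic pickaxe's reactive enemy-impact sound.
# A short fixed delay leads into an equally weighted random pick of three reactivity wave variations.
# It uses the player pickaxe impact attenuation, the shared cosmetic impact property sheet and the impacts sound class at reduced volume.
# A concurrency override allows one instance at a time, stopping the oldest when a new hit plays.
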SoundCue /BRCosmetics/Sounds/PickAxes/WeaveHarbor/PickaxeImpactEnemy_Reactive_WeaveHarbor_Cue
root=Delay_0
nodes [
 {id=Delay_0, children [Random_0], properties {delayMax=0.1, delayMin=0.1}}
 {id=Random_0, children [WavePlayer_0,WavePlayer_1,WavePlayer_2], properties {weights [1,1,1]}}
 {id=WavePlayer_0, properties {soundWaveAssetPtr=&/IP_Pickaxe_WeaveHarbor_Reactivity_01}}
 {id=WavePlayer_1, properties {soundWaveAssetPtr=&/IP_Pickaxe_WeaveHarbor_Reactivity_02}}
 {id=WavePlayer_2, properties {soundWaveAssetPtr=&/IP_Pickaxe_WeaveHarbor_Reactivity_03}}
]
properties {
 attenuationSettings=/Game/Sounds/Attenuation/Fort_Weapon_Pickaxe_Player_Impact_Att
 audioPropertiesSheet=/BRCosmetics/Sounds/Settings/PropertySheets/Pickaxe/APS_Cosmetic_Pickaxe_Impact
 bOverrideConcurrency=true
 maxDistance=3512
 soundClassObject=&Fort_Impacts
 volumeMultiplier=0.375
 concurrencyOverrides {maxCount=1, resolutionRule=StopOldest}
 concurrencySet [null]
}
