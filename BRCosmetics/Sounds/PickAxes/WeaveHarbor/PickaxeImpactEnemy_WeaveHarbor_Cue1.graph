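# Sound cue for the WeaveHarbor pickaxe's impact sound when it hits an enemy.
# A fixed short delay feeds a random node that picks one of six equally weighted impact wave variants.
# Output uses shared player pickaxe impact attenuation, a cosmetic pickaxe property sheet and the impacts sound class.
# A concurrency override allows a single instance, stopping the oldest when a new one starts.
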SoundCue /BRCosmetics/Sounds/PickAxes/WeaveHarbor/PickaxeImpactEnemy_WeaveHarbor_Cue1
root=Delay_0
nodes [
 {id=Delay_0, children [Random_0], properties {delayMax=0.1, delayMin=0.1}}
 {id=Random_0, children [WavePlayer_0,WavePlayer_1,WavePlayer_2,WavePlayer_3,WavePlayer_4,WavePlayer_5], properties {weights [1,1,1,1,1,1]}}
 {id=WavePlayer_0, properties {soundWaveAssetPtr=&/IP_Pickaxe_WeaveHarbor_Impact_01}}
 {id=WavePlayer_1, properties {soundWaveAssetPtr=&/IP_Pickaxe_WeaveHarbor_Impact_02}}
 {id=WavePlayer_2, properties {soundWaveAssetPtr=&/IP_Pickaxe_WeaveHarbor_Impact_03}}
 {id=WavePlayer_3, properties {soundWaveAssetPtr=&/IP_Pickaxe_WeaveHarbor_Impact_04}}
 {id=WavePlayer_4, properties {soundWaveAssetPtr=&/IP_Pickaxe_WeaveHarbor_Impact_05}}
 {id=WavePlayer_5, properties {soundWaveAssetPtr=&/IP_Pickaxe_WeaveHarbor_Impact_06}}
]
properties {
 attenuationSettings=/Game/Sounds/Attenuation/Fort_Weapon_Pickaxe_Player_Impact_Att
 audioPropertiesSheet=/BRCosmetics/Sounds/Settings/PropertySheets/Pickaxe/APS_Cosmetic_Pickaxe_Impact
 bOverrideConcurrency=true
 maxDistance=3512
 soundClassObject=&Fort_Impacts
 volumeMultiplier=0.51
 concurrencyOverrides {maxCount=1, resolutionRule=StopOldest, voiceStealReleaseTime=2, volumeScaleAttackTime=0.1, volumeScaleReleaseTime=1}
 concurrencySet [null]
}
